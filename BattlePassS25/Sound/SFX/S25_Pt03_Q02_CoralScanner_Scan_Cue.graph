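# Sound cue for the Coral Scanner scan sound effect in a Season 25 Battle Pass quest, assigned to the Fort_Quests sound class.
# A Random node picks between two scan wave variants with equal weight.
# A root Modulator adds slight random pitch and volume variation to the chosen wave.
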SoundCue /BattlePassS25/Sound/SFX/S25_Pt03_Q02_CoralScanner_Scan_Cue
root=Modulator_0
nodes [
 {id=Modulator_0, children [Random_1], properties {pitchMax=1.02, pitchMin=0.98, volumeMax=1.01, volumeMin=1}}
 {id=Random_1, children [WavePlayer_1,WavePlayer_2], properties {weights [1,1]}}
 {id=WavePlayer_1, properties {soundWaveAssetPtr=&/S25_Pt3_Q02_CoralScanner_Scan_01}}
 {id=WavePlayer_2, properties {soundWaveAssetPtr=&/S25_Pt3_Q02_CoralScanner_Scan_02}}
]
properties {maxDistance=2097152, soundClassObject=&Fort_Quests, volumeMultiplier=1}
