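# Quest sound cue for collecting a Coral Buddy in the Season 25 Battle Pass, routed to the Fort_Quests sound class.
# A mixer layers a random TeriyakiFish cheer at full volume, a quieter random cage-close variant, and a briefly delayed, still quieter reward-screen chest upgrade sound.
# A root modulator adds slight pitch and volume variation to the combined mix.
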SoundCue /BattlePassS25/Sound/SFX/S25_Pt04_Q04_CoralBuddy_Collect_Cue
root=Modulator_0
nodes [
 {id=Delay_0, children [WavePlayer_17], properties {delayMax=0.1, delayMin=0.1}}
 {id=Mixer_1, children [Random_0,Random_2,Delay_0], properties {inputVolume [1,0.64,0.28]}}
 {id=Modulator_0, children [Mixer_1], properties {pitchMax=1.02, pitchMin=0.98, volumeMax=1.01, volumeMin=1}}
 {
  id=Random_0
  children [
   WavePlayer_0
   WavePlayer_3
   WavePlayer_4
   WavePlayer_5
   WavePlayer_6
   WavePlayer_7
   WavePlayer_8
   WavePlayer_9
   WavePlayer_10
   WavePlayer_11
   WavePlayer_12
   WavePlayer_13
  ]
  properties {weights [1,1,1,1,1,1,1,1,1,1,1,1]}
 }
 {id=Random_2, children [WavePlayer_14, WavePlayer_15, WavePlayer_16], properties {weights [1,1,1]}}
 {id=WavePlayer_0, properties {soundWaveAssetPtr=&/Glider_TeriyakiFish_Cheers_11}}
 {id=WavePlayer_10, properties {soundWaveAssetPtr=&/Glider_TeriyakiFish_Cheers_07}}
 {id=WavePlayer_11, properties {soundWaveAssetPtr=&/Glider_TeriyakiFish_Cheers_08}}
 {id=WavePlayer_12, properties {soundWaveAssetPtr=&/Glider_TeriyakiFish_Cheers_09}}
 {id=WavePlayer_13, properties {soundWaveAssetPtr=&/Glider_TeriyakiFish_Cheers_10}}
 {id=WavePlayer_14, properties {soundWaveAssetPtr=&/S25_Q04_Cage_Close_01}}
 {id=WavePlayer_15, properties {soundWaveAssetPtr=&/S25_Q04_Cage_Close_02}}
 {id=WavePlayer_16, properties {soundWaveAssetPtr=&/S25_Q04_Cage_Close_03}}
 {id=WavePlayer_17, properties {soundWaveAssetPtr=/Game/Sounds/Fort_UI_Sounds/Rewards/Fort_RewardScreen_TeamScore_ChestLootUpgrade}}
 {id=WavePlayer_3, properties {soundWaveAssetPtr=&/Glider_TeriyakiFish_Cheers_00}}
 {id=WavePlayer_4, properties {soundWaveAssetPtr=&/Glider_TeriyakiFish_Cheers_01}}
 {id=WavePlayer_5, properties {soundWaveAssetPtr=&/Glider_TeriyakiFish_Cheers_02}}
 {id=WavePlayer_6, properties {soundWaveAssetPtr=&/Glider_TeriyakiFish_Cheers_03}}
 {id=WavePlayer_7, properties {soundWaveAssetPtr=&/Glider_TeriyakiFish_Cheers_04}}
 {id=WavePlayer_8, properties {soundWaveAssetPtr=&/Glider_TeriyakiFish_Cheers_05}}
 {id=WavePlayer_9, properties {soundWaveAssetPtr=&/Glider_TeriyakiFish_Cheers_06}}
]
properties {maxDistance=2097152, soundClassObject=&Fort_Quests, volumeMultiplier=1}
