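# Sound cue for a looping Coral Buddies singing effect, named for a Season 25 Battle Pass quest.
# A single looping wave player feeds the root directly, with no randomization or mixing.
# The cue routes to the quest sound class and overrides attenuation with its own falloff and a wide stereo spread for localized playback in the world.
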SoundCue /BattlePassS25/Sound/SFX/S25_Pt04_Q04_CoralChorus_Cue
root=WavePlayer_14
nodes [{id=WavePlayer_14, properties {bLooping=true, soundWaveAssetPtr=&/S25_Q04_CoralBuddiesSinging_LP}}]
properties {
 bOverrideAttenuation=true
 maxDistance=1050
 soundClassObject=&Fort_Quests
 volumeMultiplier=1
 attenuationOverrides {falloffDistance=850, stereoSpread=2300, attenuationShapeExtents {x=200}}
}
