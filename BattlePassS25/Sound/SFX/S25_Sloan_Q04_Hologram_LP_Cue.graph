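# Looping hologram sound effect cue for the Season 25 Battle Pass Sloan Q04 quest, routed through the Fort_Quests sound class.
# A looping wave player feeds a modulator that slightly randomizes pitch, then an enveloper that applies a short volume fade-in.
# The cue raises overall pitch and takes its spatial falloff from the Part03 Q04 quest attenuation asset.
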SoundCue /BattlePassS25/Sound/SFX/S25_Sloan_Q04_Hologram_LP_Cue
root=Enveloper_0
nodes [
 {
  id=Enveloper_0
  children [Modulator_0]
  properties {pitchCurve {editorCurveData {keys [{time=0}]}}, volumeCurve {editorCurveData {keys [<at 0 0>,<at 0.1 1>]}}}
 }
 {id=Modulator_0, children [WavePlayer_0], properties {pitchMax=1.02, pitchMin=0.98}}
 {id=WavePlayer_0, properties {bLooping=true, soundWaveAssetPtr=&/S25_Sloan_Q04_Hologram_LP}}
]
properties {
 attenuationSettings=/BattlePassS25/Items/QuestActors/Stories/Part03/SFX/S25_P03_Q04_Atten
 maxDistance=1100
 pitchMultiplier=1.2
 soundClassObject=&Fort_Quests
 volumeMultiplier=1
}
