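# Looping gem dig sound cue for a Season 25 Battle Pass Slone quest (Q03), in the quests sound class.
# A mixer layers two looping Telescope gem waves, the first at reduced level, under a modulator that slightly randomizes volume.
# The cue overrides attenuation with its own short-range falloff, though it also references a Q04 attenuation asset.
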SoundCue /BattlePassS25/Sound/SFX/S25_Slone_Q03_Gem_Dig_LP_Cue
root=Modulator_2
nodes [
 {id=Mixer_0, children [WavePlayer_0,WavePlayer_1], properties {inputVolume [0.7,1]}}
 {id=Modulator_2, children [Mixer_0], properties {volumeMax=0.7, volumeMin=0.6}}
 {id=WavePlayer_0, properties {bLooping=true, soundWaveAssetPtr=&/Telescope_Gem_Loop_01}}
 {id=WavePlayer_1, properties {bLooping=true, soundWaveAssetPtr=&/Telescope_Gem_Loop_02}}
]
properties {
 attenuationSettings=&/S25_Sloan_Q04_Atten
 bOverrideAttenuation=true
 maxDistance=1050
 soundClassObject=&Fort_Quests
 volumeMultiplier=0.8
 attenuationOverrides {falloffDistance=850, attenuationShapeExtents {x=200}}
 concurrencyOverrides {maxCount=8}
}
