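# Sound cue for the Season 25 Slone Q04 gem delivery, classed as a quest sound.
# A mixer layers three branches: a randomly chosen success sound at boosted volume, a telescope gem-deliver sound after a short fixed delay, and a random pick from six "Lyr" variations.
# Both random picks are equally weighted, and an attenuation asset overrides the cue's attenuation to spatialize playback.
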SoundCue /BattlePassS25/Sound/SFX/S25_Slone_Q04_Gem_Deliver_Cue
root=Mixer_0
nodes [
 {id=Delay_0, children [WavePlayer_0], properties {delayMax=0.4, delayMin=0.4}}
 {id=Mixer_0, children [Random_1,Delay_0,Random_0], properties {inputVolume [1.5,1,1]}}
 {id=Random_0, children [WavePlayer_1,WavePlayer_2,WavePlayer_3,WavePlayer_4,WavePlayer_5,WavePlayer_6], properties {weights [1,1,1,1,1,1]}}
 {id=Random_1, children [WavePlayer_8,WavePlayer_7], properties {weights [1,1]}}
 {id=WavePlayer_0, properties {soundWaveAssetPtr=&/Telescope_Gem_Deliver_01}}
 {id=WavePlayer_1, properties {soundWaveAssetPtr=&/S25_Sloan_Q04_Gem_Deliver_Lyr_01}}
 {id=WavePlayer_2, properties {soundWaveAssetPtr=&/S25_Sloan_Q04_Gem_Deliver_Lyr_02}}
 {id=WavePlayer_3, properties {soundWaveAssetPtr=&/S25_Sloan_Q04_Gem_Deliver_Lyr_03}}
 {id=WavePlayer_4, properties {soundWaveAssetPtr=&/S25_Sloan_Q04_Gem_Deliver_Lyr_04}}
 {id=WavePlayer_5, properties {soundWaveAssetPtr=&/S25_Sloan_Q04_Gem_Deliver_Lyr_05}}
 {id=WavePlayer_6, properties {soundWaveAssetPtr=&/S25_Sloan_Q04_Gem_Deliver_Lyr_06}}
 {id=WavePlayer_7, properties {soundWaveAssetPtr=&/S25_Sloan_Q04_Gem_Deliver_Success_02}}
 {id=WavePlayer_8, properties {soundWaveAssetPtr=&/S25_Sloan_Q04_Gem_Deliver_Success_01}}
]
properties {
 attenuationSettings=&/S25_Sloan_Q04_Atten
 bOverrideAttenuation=true
 maxDistance=4000
 soundClassObject=&Fort_Quests
}
